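# Widget Blueprint for the Mash playlist's dance pad, showing an "Emote to Heal!" prompt in uppercase.
# An Overlay layers a UIKit backplate behind a centered row holding a keybind icon, a spacer and outlined text.
# The keybind widget binds to the EmotePicker action, with a gamepad emote picker/replay-hold fallback.
# The layout is static: the event graph holds only empty event stubs and no variables are declared.
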WidgetBlueprint /Game/Athena/Playlists/Mash/UI/WBP_Mash_DancePad_EmotePromptUI
graphs [EventGraph {nodes [
   Event PreConstruct()
   Event Construct()
   Event Tick()
  ]}]
variables []
widgets {
 info {inheritedWidgetCount=0, parentClass=&UserWidget, rootWidgetClass=&Overlay, widgetCount=6}
 widgets [
  Overlay_44 {widgetClassPath=&Overlay, widget=Overlay_44}
  WBP_UIKit_Backplate {
   widgetClassPath=&WBP_UIKit_Backplate
   bIsVariable=true
   parent=Overlay_44
   widget=WBP_UIKit_Backplate
   properties {brightness=Medium}
   slot {class=OverlaySlot, properties {horizontalAlignment=HAlign_Fill, verticalAlignment=VAlign_Fill}}
  }
  HorizontalBox_0 {
   widgetClassPath=&HorizontalBox
   bIsVariable=true
   parent=Overlay_44
   widget=HorizontalBox_0
   slot {class=OverlaySlot, properties {horizontalAlignment=HAlign_Center, verticalAlignment=VAlign_Center, padding=<ltrb 20 10 20 10>}}
  }
  KeybindWidget_Emote {
   widgetClassPath=/FortUILibrary/Input/KeybindWidget
   bIsVariable=true
   parent=HorizontalBox_0
   widget=KeybindWidget_Emote
   properties {bUseKeyboardTextureForKeys=true, boundAction=EmotePicker, boundActionFallbacks [GamepadEmotePickerOrReplayLastEmoteHold]}
   slot {class=HorizontalBoxSlot, properties {horizontalAlignment=HAlign_Center, verticalAlignment=VAlign_Center}}
  }
  Spacer_143 {widgetClassPath=&Spacer, parent=HorizontalBox_0, widget=Spacer_143, properties {size=<8 1>}, slot {class=HorizontalBoxSlot}}
  PromptText {
   widgetClassPath=&CommonTextBlock
   bIsVariable=true
   parent=HorizontalBox_0
   widget=PromptText
   properties {
    text="Emote to Heal!"
    textTransformPolicy=ToUpper
    font {
     bForceMonospaced=false
     bMaterialIsStencil=false
     fontMaterial=null
     fontObject=&HeadingNow
     letterSpacing=-1
     monospacedWidth=1
     size=21
     skewAmount=0
     typefaceFontName=76 Bold
     outlineSettings {
      bApplyOutlineToDropShadows=false
      bMiteredCorners=false
      bSeparateFillAlpha=false
      outlineMaterial=&M-FontShadow
      outlineSize=2
      outlineColor=<rgba 0 0 0 1>
     }
    }
    shadowOffset=<0 0>
   }
   slot {class=HorizontalBoxSlot}
  }
 ]
}
