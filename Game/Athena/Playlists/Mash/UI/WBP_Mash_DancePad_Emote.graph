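# Widget Blueprint for the Mash playlist's dance pad that displays an emote prompt on screen.
# It is a thin wrapper: a canvas panel holding one WBP_Mash_DancePad_EmotePromptUI, anchored to screen center and offset below it.
# The PreConstruct, Construct and Tick events are empty and no variables are declared, so this widget adds no logic of its own.
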WidgetBlueprint /Game/Athena/Playlists/Mash/UI/WBP_Mash_DancePad_Emote
graphs [EventGraph {nodes [
   Event PreConstruct()
   Event Construct()
   Event Tick()
  ]}]
variables []
widgets {
 info {inheritedWidgetCount=0, parentClass=&UserWidget, rootWidgetClass=&CanvasPanel, widgetCount=2}
 widgets [
  CanvasPanel_19 {widgetClassPath=&CanvasPanel, widget=CanvasPanel_19}
  WBP_Mash_DancePad_EmotePromptUI {
   widgetClassPath=&/WBP_Mash_DancePad_EmotePromptUI
   bIsVariable=true
   parent=CanvasPanel_19
   widget=WBP_Mash_DancePad_EmotePromptUI
   slot {
    class=CanvasPanelSlot
    properties {layoutData {alignment=<0 0>, anchors {maximum=<0.5 0.5>, minimum=<0.5 0.5>}, offsets=<ltrb -180 299.4595 358.1081 60.6306>}}
   }
  }
 ]
}
